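# Sound cue for the third-person foley of the SmallFry emote, a flying sound with a start and a sustained loop.
# A Mixer layers a one-shot start wave, passed through a Modulator, with a looping wave shaped by an Enveloper.
# The envelope fades the loop in over one second while its pitch briefly rises and settles, blending it under the start.
# Emote foley attenuation, class and concurrency settings govern playback.
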SoundCue /BRCosmetics/Sounds/Emotes/SmallFry/EmoteFoley_SmallFry_3P_Cue
root=Mixer_0
nodes [
 {
  id=Enveloper_0
  children [WavePlayer_1]
  properties {
   pitchCurve {editorCurveData {keys [
      {arriveTangent=0.48, interpMode=RCIM_Cubic, leaveTangent=0.48, tangentMode=RCTM_User, time=0}
      {arriveTangent=0.0675, interpMode=RCIM_Cubic, leaveTangent=0.0675, tangentMode=RCTM_User, time=0.252, value=1.1884}
      {arriveTangent=-0.0207, interpMode=RCIM_Cubic, leaveTangent=-0.0207, tangentMode=RCTM_User, time=2}
     ]}}
   volumeCurve {editorCurveData {keys [
      {arriveTangent=2.6269, interpMode=RCIM_Cubic, leaveTangent=2.6269, tangentMode=RCTM_User, time=0, value=0}
      {arriveTangent=8.925, interpMode=RCIM_Cubic, leaveTangent=8.9251, tangentMode=RCTM_User, time=0.2319, value=0.434}
      <at 1 1>
     ]}}
  }
 }
 {id=Mixer_0, children [Modulator_0,Enveloper_0], properties {inputVolume [1,1]}}
 {id=Modulator_0, children [WavePlayer_0]}
 {id=WavePlayer_0, properties {soundWaveAssetPtr=&/EmoteFoley_SmallFry_Flying_Loop_Start_01}}
 {id=WavePlayer_1, properties {bLooping=true, soundWaveAssetPtr=&/EmoteFoley_SmallFry_Flying_Loop}}
]
properties {
 attenuationSettings=&Fort_Emotes_Athena_Att
 audioPropertiesSheet=/BRCosmetics/Sounds/Settings/PropertySheets/Emotes/APS_Cosmetic_Emote_Foley
 maxDistance=1600
 soundClassObject=/Game/Sounds/FortSoundClassesAndModes/Fort_Foley_Emote
 volumeMultiplier=0.8
 concurrencyOverrides {maxCount=3}
 concurrencySet [/Game/Sounds/Concurrency_Rules/Athena_EmoteFoley_Conc]
 soundSubmixSends [{minSendLevel=0, soundSubmix=&Fortnite_AudioAnalysis_Music}]
}
